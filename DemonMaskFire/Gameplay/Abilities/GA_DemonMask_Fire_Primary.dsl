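; Primary fire ability for the Fire Demon Mask: launches a projectile and ends, first playing a mask toss montage if ammo is depleted.
; The server spawns the projectile along the player's aim, line-tracing to pull the spawn point back from blocking geometry.
; A box trace picks the visible non-teammate nearest the reticle, with a cloak-tag check.
; That homing target is stored on the weapon for the projectile, avoiding a BeginPlay race.
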(event Ability|EventActivateAbility
  (|CallParentFunction)
  (bind _asfort_player_pawn (Utilities|Casting|CastToFortPlayerPawn (Ability|GetAvatarActorfromActorInfo)))
  (bind _output_get (Variables|Default|SetAs%20Fort%20Player%20Pawn _asfort_player_pawn))
  (bind _asb_demon_mask_fire (Utilities|Casting|CastToB_DemonMask_Fire (Variables|/Script/FortniteGame.FortPawn|GetCurrentWeapon _output_get))
    (:then
      (Variables|Default|SetAs%20B%20Demon%20Mask%20Fire _asb_demon_mask_fire)
      (if (Networking|IsServer)
        (bind _returnvalue (Ability|CommitAbility))
        (if (not _returnvalue)
          (Ability|EndAbility))
        (else
          _))
      (Default|FireSingle)
      (CallFunction|FindandSetHomingTarget)
      ;[1 0.9695 0 0.5] TODO: for some reason. the fire mask ammo tracking GA is not consistently firing from the mask weapon event to player... no idea why. Void mask and basis item fringe plank work fine
      ; plugging the same style of logic in here on fire projectile
      (Utilities|FlowControl|Delay 0.2)
      (if (<= (FortWeapon|GetRemainingAmmo (Variables|Default|GetAs%20B%20Demon%20Mask%20Fire)) 0)
        (bind (_asynctaskproxy _notifytag) (Ability|Tasks|PlayMontageAndWaitNotify "None" "/DemonMaskFire/Assets/Anim/DemonMaskFire_Toss_M.DemonMaskFire_Toss_M" 1.0 "None" false)
          (:OnCompleted
            _)
          (:OnInterrupted
            _)
          (:OnCancelled
            _)
          (:OnNotifyBegin
            (switch Utilities|FlowControl|Switch|SwitchonGameplayTag _notifytag
              (:Event.Athena.Item.FringePlank.Toss
                (Utilities|FlowControl|Delay 0.5)
                (Ability|EndAbility))
              (:Event.Athena.Item.DemonMaskFire.TossMaskMaterial
                (Class|BPDemonMaskFireMaskPartModifier|DemonMaskFireMaterialWipeonToss (Variables|/DemonMaskFire/Gameplay/Actors/B_DemonMask_Fire.B_DemonMask_Fire_C|GetPartModifier (Variables|Default|GetAs%20B%20Demon%20Mask%20Fire)))))))
        (else
          (Ability|EndAbility))))
    (:CastFailed
      (Ability|EndAbility))))

(fn CalcProjectileDirection
  (params (Start :type Vector%20%28by%20ref%29) (IgnoreSpread :type Boolean))
  (bind _returnvalue (Ability|GetAvatarActorfromActorInfo))
  (bind _returnvalue_2 (Fortnite|GetFortPlayerControllerfromActor _returnvalue))
  (if (Utilities|IsValid _returnvalue_2 :!class /Script/Engine.KismetSystemLibrary)
    ;[0.8694 0.8958 0 0.5] Aiming near a Sideways / Demon Arena bubble was causing GetPlayerAimPoint to fail without checking No Trace
    ; TODO: Why is this function trace failing near these??
    (bind _returnvalue_3 (Fortnite|GetPlayerAimPoint _returnvalue_2 6e+03 (Default|GetAimIgnoredActors) true))
    (return (Math|Rotator|SelectRotator (Math|Rotator|FindLookatRotation Start _returnvalue_3) (Math|Rotator|MakeRotator (.roll (Math|Rotator|Lerp(Rotator) (Transformation|GetActorRotation _returnvalue) (Math|Rotator|FindLookatRotation Start _returnvalue_3) (ScalableFloat|GetValueatLevel (Variables|Firing|GetRow_Accuracy)) true)) (.pitch (Math|Rotator|FindLookatRotation Start _returnvalue_3)) (.yaw (Math|Rotator|Lerp(Rotator) (Transformation|GetActorRotation _returnvalue) (Math|Rotator|FindLookatRotation Start _returnvalue_3) (ScalableFloat|GetValueatLevel (Variables|Firing|GetRow_Accuracy)) true))) (Variables|GetIgnoreSpread)))
    (else
      (bind _outfortplayerpawn (CallFunction|GetAvatarasValidFortPlayerPawn)
        (:ValidFortPlayerPawn
          (return (Targeting|GetDefaultTargetingRotation _outfortplayerpawn)))
        (:AvatarCastFailed
          (Ability|EndAbility))))))

(fn FindAndSetHomingTarget
  (bind _as_b_demon_mask_fire (Variables|Default|GetAs%20B%20Demon%20Mask%20Fire))
  (if (ScalableFloat|GetValueasBool (Variables|Projectile|GetRow_HomingEnabled))
    ;[0 0.5834 0.9635 0.5] Recreation of GetAimedAtEnemy(), but allowing non-FortPawns to be returned. Includes AI Pawns by default.
    ; The previous GetAimedAtEnemy() node had "Target AI Pawns" true at all times.
    (bind _returnvalue (Default|GetAimedAtTarget (Fortnite|GetFortPlayerControllerfromActor (Variables|Default|GetAs%20Fort%20Player%20Pawn)) (ScalableFloat|GetValueatLevel (Variables|Default|GetDataHomingRectHeightandWidth)) (ScalableFloat|GetValueatLevel (Variables|Default|GetDataHomingRectHeightandWidth)) (ScalableFloat|GetValueatLevel (Variables|Default|GetDataHomingRange))))
    ;[1 1 1] Don't allow homing on cloaked targets
    (Utilities|IsValid _returnvalue
      (:Is%20Valid
        (bind _asfort_player_pawn (Utilities|Casting|CastToFortPlayerPawn _returnvalue)
          (:then
            (if (GameplayTags|HasAnyMatchingGameplayTags _asfort_player_pawn (Variables|Default|GetAvoid%20Detection%20Tags%20Container))
              ;[1 1 1] Store Homing target on Weapon, projectile can pull from there. Trying to set directly on projectile after spawning it has race condition issues with projectile Begin Play
              (Variables|/DemonMaskFire/Gameplay/Actors/B_DemonMask_Fire.B_DemonMask_Fire_C|SetHomingTarget 0 _as_b_demon_mask_fire)
              (else
                _)))
          (:CastFailed
            _))
        (Variables|/DemonMaskFire/Gameplay/Actors/B_DemonMask_Fire.B_DemonMask_Fire_C|SetHomingTarget _returnvalue _as_b_demon_mask_fire)))))

(fn FireSingle
  (Ability|HasAuthority
    (:True
      (Default|SpawnProjectile (Default|GetProjectileSpawnLocation)))))

(fn GetAimIgnoredActors
  (return (select (Utilities|IsValid (Variables|Default|GetAs%20B%20Demon%20Mask%20Fire) :!class /Script/Engine.KismetSystemLibrary) (Utilities|Array|MakeArray (Ability|GetAvatarActorfromActorInfo) (Variables|Default|GetAs%20B%20Demon%20Mask%20Fire)) (Utilities|Array|MakeArray (Ability|GetAvatarActorfromActorInfo)))))

(fn GetAimedatTarget
  (params (SourcePlayer :type Fort%20Player%20Controller%20Object%20Reference) (RectHalfWidth :type Float%20%28double-precision%29) (RectHalfHeight :type Float%20%28double-precision%29) (Range :type Float%20%28double-precision%29))
  (bind _sourceplayer (Variables|GetSourcePlayer))
  (if (Utilities|IsValid SourcePlayer :!class /Script/Engine.KismetSystemLibrary)
    (if (and (not (Pawn|IsLocalController SourcePlayer)) (Utilities|Enum|NotEqual(Enum) (Networking|GetLocalRole SourcePlayer) "ROLE_Authority"))
      (return 0)
      (else
        (LocalVariables|SetCameraPosition (Pawn|GetPlayerViewPoint _sourceplayer))
        (LocalVariables|SetCameraRotation (Pawn|GetBaseAimRotation (Pawn|GetControlledPawn _sourceplayer)))
        (LocalVariables|SetBestDistance (+ (Math|Float|Sqrt (+ (Math|Float|Square (Variables|GetRectHalfHeight)) (Math|Float|Square (Variables|GetRectHalfWidth)))) 0.0001))
        (LocalVariables|SetForwardAim (Math|Vector|GetForwardVector (LocalVariables|GetCameraRotation)))
        ;[0.4031 1 0.709 0.5] Find potential targets
        ; Start position is moved forward to where the player pawn is, otherwise we actually wind up targeting behind us as the box trace extends 2x the Rect width behind
        (bind (_outhits _returnvalue) (Collision|MultiBoxTraceForObjects (Utilities|Operators|Add :B (LocalVariables|GetCameraPosition) :C (* (LocalVariables|GetForwardAim) (* (Variables|GetRectHalfWidth) 1.0))) (+ (LocalVariables|GetCameraPosition) (* (LocalVariables|GetForwardAim) (Variables|GetRange))) (Math|Vector|MakeVector (Variables|GetRectHalfWidth) (Variables|GetRectHalfHeight) 1e+01) (LocalVariables|GetCameraRotation) (Utilities|Array|MakeArray "ObjectTypeQuery3" "ObjectTypeQuery7") false (Variables|Default|GetActorsToIgnoreForTargeting) "None" true "(R=1.000000,G=0.000000,B=0.000000,A=1.000000)" "(R=0.000000,G=1.000000,B=0.000000,A=1.000000)" 5.0))
        (if _returnvalue
          (LocalVariables|SetNewLocalVar_0 _outhits)
          (for _array_element (LocalVariables|GetNewLocalVar_0)
            (bind _ascharacter (Utilities|Casting|CastToCharacter (.HitActor (Collision|BreakHitResult _array_element))))
            (if (not (Team|OnSameTeam (Variables|GetSourcePlayer) _ascharacter))
              ;[0.6979 0.7664 1 0.5] Is this target closer to our reticle than the current best candidate?
              (bind _output_get (LocalVariables|SetPotentialDistance (Math|Vector|GetPointDistancetoLine (Transformation|GetActorLocation (.HitActor (Collision|BreakHitResult _array_element))) (LocalVariables|GetCameraPosition) (LocalVariables|GetForwardAim))))
              (if (< _output_get (LocalVariables|GetBestDistance))
                (bind _returnvalue_2 (Fortnite|CheckLineOfSightToActorWithChannel (LocalVariables|GetCameraPosition) (.HitActor (Collision|BreakHitResult _array_element)) "ECC_GameTraceChannel16" (Variables|GetSourcePlayer))) ; Check if there's anything in the way, like a wall.\nOnly do this raycast if the other checks have been passed.
                (if _returnvalue_2
                  (LocalVariables|SetBestTarget (.HitActor (Collision|BreakHitResult _array_element)))
                  (LocalVariables|SetBestDistance (LocalVariables|GetPotentialDistance))))))
          (return (LocalVariables|GetBestTarget)))))
    (else
      (return 0))))

(fn GetProjectileSpawnDirection
  (bind _outfortplayerpawn (CallFunction|GetAvatarasValidFortPlayerPawn)
    (:ValidFortPlayerPawn
      (return (Pawn|GetControlRotation _outfortplayerpawn :!class /Script/Engine.Pawn)))))

(fn GetProjectileSpawnLocation
  (bind _verbosedebug (Variables|Default|GetVerboseDebug))
  (bind _outloc (LocalVariables|GetOutLoc))
  (Utilities|FlowControl|Sequence
    (:then_0
      (bind _output_get (LocalVariables|SetOutLoc (+ (Math|Vector|RotateVector (Math|Vector|SelectVector (Variables|Default|GetRidingOffset) (select (Variables|/Script/Engine.Character|GetbIsCrouched (Variables|Default|GetAs%20Fort%20Player%20Pawn)) (Variables|Default|GetCrouchOffset) (Variables|Default|GetStandardOffset)) (GameplayTags|HasMatchingGameplayTag (Variables|Default|GetAs%20Fort%20Player%20Pawn) "(TagName=\"Granted.Riding.Player.IsRiding\")")) (Transformation|GetActorRotation (Variables|Default|GetAs%20Fort%20Player%20Pawn))) (Transformation|GetActorLocation (Variables|Default|GetAs%20Fort%20Player%20Pawn)))))
      (bind _as_fort_player_pawn (Variables|Default|GetAs%20Fort%20Player%20Pawn)
        (:then
          ;[0.573 1 0.8961] Line trace from the actor location to the projectile spawn location to check for any "missed hits". If something is hit, adjust the projectile location so that it is before that hit.
          ; This is necessary since the projectile spawn offset is far enough from the player's location that it can miss hitting something if the player is right up against it.
          (bind (_outhit _returnvalue) (Collision|LineTraceByProfile (Transformation|GetActorLocation _as_fort_player_pawn) _outloc "FortProjectileThrownOverlapPawns" false (Utilities|Array|MakeArray _as_fort_player_pawn (Variables|/Script/FortniteGame.FortPawn|GetCurrentWeapon _as_fort_player_pawn) (Actor|GetAttachParentActor _as_fort_player_pawn)) "None" true "(R=1.000000,G=0.000000,B=0.000000,A=1.000000)" "(R=0.000000,G=1.000000,B=0.000000,A=1.000000)" 1e+01)) ; Do NOT worry about pawns  here. The projectile has a much larger pawn specific collision,\nso this issue does not occur when throwing at a close pawn. 
          (if _returnvalue
            (LocalVariables|SetOutLoc (+ (* (Math|Vector|Normalize (- (Transformation|GetActorLocation _as_fort_player_pawn) _outloc)) (LocalVariables|GetAdjustmentHitOffset)) (.Location (Collision|BreakHitResult _outhit))))
            (Development|PrintString (Utilities|String|Append "Start location trace - Blocking hit, adjusting start location - " (Utilities|GetDisplayName (.HitActor (Collision|BreakHitResult _outhit)))))
            (else
              (Utilities|IsValid (.HitActor (Collision|BreakHitResult _outhit))
                (:Is%20Valid
                  (Development|PrintString (Utilities|String|Append "Start location trace - non blocking hit - " (Utilities|GetDisplayName (.HitActor (Collision|BreakHitResult _outhit)))) _verbosedebug _verbosedebug))
                (:Is%20Not%20Valid
                  (Development|PrintString "Start location trace - No hit" _verbosedebug _verbosedebug))))))
        (:else)))
    (:then_1
      (return (LocalVariables|GetOutLoc)))))

(fn CommitExecute
  (bind _returnvalue (Ability|GetAvatarActorfromActorInfo))
  ;[1 1 1] Send a message to the shooter so that they do the fire traces for placement
  (Utilities|IsValid _returnvalue
    (:Is%20Valid
      (Ability|SendGameplayEventtoActor _returnvalue "(TagName=\"Event.Weapon.Fired\")"))
    (:Is%20Not%20Valid
      _))
  (|CallParentFunction))

(fn SpawnProjectile
  (params (InitialLocation :type Vector) (ProjectileIndex :type Integer))
  (bind _returnvalue (Default|CalcProjectileDirection InitialLocation true))
  (Projectile|SpawnProjectilewithParams (Variables|Projectile|GetProjectile) (Ability|GetAvatarActorfromActorInfo) (Utilities|Struct|MakeSpawnProjectileParams InitialLocation (Transformation|GetActorRotation (Ability|GetAvatarActorfromActorInfo)) _returnvalue "(TargetSelection=(List=,bStopAtFirstSuccess=True,bKeepCheckingListOnIndestructibleHit=False,AbilityTargetBlockedGameplayTags=(GameplayTags=),bUseWeaponRanges=True,bUseMaxYawAngleToTarget=False,MaxYawAngleToTarget=0.000000),TargetGameplayEffectSpecs=,OwnerGameplayEffectSpecs=,ActivationCues=(GameplayTags=),ImpactCues=(GameplayTags=),ImpactNoiseRange=0.000000,FlyByNoiseRange=0.000000,bOverrideChargeMagnitude=False,ChargeMagnitudeOverrideValue=0.000000)" (Ability|MakeGameplayEffectContainerSpec (Variables|ApplicationTags|GetExplode%20Application%20Tag)) false 0 (Math|Random|RandomFloatinRange (ScalableFloat|GetValueatLevel (Variables|Projectile|GetRow_InitialSpeedMin)) (ScalableFloat|GetValueatLevel (Variables|Projectile|GetRow_InitialSpeedMax))) (Math|Random|RandomFloatinRange (ScalableFloat|GetValueatLevel (Variables|Projectile|GetRow_InitialSpeedMin)) (ScalableFloat|GetValueatLevel (Variables|Projectile|GetRow_InitialSpeedMax))) 0.0)))
